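# M_ScreenMirror is a UI-domain material in the frontend live-stream video folder that references a Season 5 Papaya media player video texture.
# Its graph holds a single SpriteTextureSampler node with that texture as its default.
# No connection from the sampler to a material output appears in the graph.
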Material /Game/Athena/UI/Frontend/LiveStreamVIdeo/M_ScreenMirror
graph {
 properties {bUsedWithStaticMesh=false, materialDomain=MD_UI, pixelDepthOffsetMode=PDOM_Legacy, usedShadingModels=Default Lit}
 nodes [SpriteTextureSampler!(UVs, Apply View MipBias) {texture=/Game/Movies/Season5/Papaya_MediaPlayer_Video1}]
}
